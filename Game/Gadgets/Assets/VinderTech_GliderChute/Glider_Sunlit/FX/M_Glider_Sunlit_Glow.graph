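# Translucent, unlit, two-sided glow material for Niagara sprites and mesh particles in the Sunlit glider's FX.
# A shape mask blends from a procedural diamond gradient to a powered sphere mask or a sparkle texture via scalar parameters.
# That mask lerps two colors for particle-tinted emissive, passed through an emissive-adjustment function, and also drives opacity.
# Opacity includes particle alpha and view-angle fade, with depth fade skipped on the mobile shading path.
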Material /Game/Gadgets/Assets/VinderTech_GliderChute/Glider_Sunlit/FX/M_Glider_Sunlit_Glow
graph {
 properties {
  bUsedWithNiagaraMeshParticles=true
  bUsedWithNiagaraSprites=true
  blendMode=BLEND_Translucent
  pixelDepthOffsetMode=PDOM_Legacy
  shadingModel=MSM_Unlit
  twoSided=true
  usedShadingModels=Unlit
 }
 nodes [
  n1 = Multiply(
   A=MaterialFunctionCall!(Falloff=ScalarParameter() {defaultValue=2.5, parameterName=Diamond Power}) {materialFunction=/Engine/Functions/Engine_MaterialFunctions01/Gradient/DiamondGradient}.DiamondGradient
   B=ScalarParameter() {defaultValue=1, parameterName=Diamond Mult}
  )
  n2 = SphereMask(A=TextureCoordinate(), B=Constant() {r=0.5}, Radius=ScalarParameter() {defaultValue=0.4, parameterName=Sphere Radius}) {attenuationRadius=0.5, hardnessPercent=0}
  n3 = ScalarParameter() {defaultValue=3, parameterName=Sphere Power}
  n4 = LinearInterpolate(
   A=LinearInterpolate(A=Saturate(None=@n1), B=Power(Base=@n2, Exp=@n3), Alpha=ScalarParameter() {parameterName=Lerp Sphere})
   B=TextureSampleParameter2D() {parameterName=BaseTexture, texture=/Game/UI/Foundation/Textures/Pattern/T_Sparkle}.RGB
   Alpha=ScalarParameter() {parameterName=Use_BaseTexture}
  )
  n5 = LinearInterpolate(
   A=VectorParameter() {parameterName=Color A, defaultValue=<rgba 0 0.0833 1 1>}.RGB
   B=VectorParameter() {parameterName=Color B, defaultValue=<rgba 0 0.1667 1 1>}.RGB
   Alpha=@n4
  )
  n6 = ScalarParameter() {defaultValue=3, parameterName=Emissive Mult}
  n7 = ParticleColor()
  MaterialFunctionCall(Emissive=Multiply(A=Multiply(A=@n5, B=@n6), B=@n7.RGB)) {materialFunction=&MF_FX_EmissiveAdjustments}
  n8 = Multiply(
   A=Multiply(A=Multiply(A=@n7.A, B=@n4), B=ScalarParameter() {defaultValue=1, parameterName=Opacity Mult})
   B=MaterialFunctionCall("FadeSensitivity (optional)"=ScalarParameter() {defaultValue=1, group=Alpha, parameterName=FadeViewAngle}) {materialFunction=&FadeBasedOnViewAngle}.Result
  )
  n9 = ScalarParameter() {defaultValue=10, group=Alpha, parameterName=FadeDistance}
  ShadingPathSwitch(Default=Multiply(A=@n8, B=DepthFade(FadeDistance=@n9)), Mobile=@n8) {desc=Migrated from FeatureLevelSwitch}
 ]
}
